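# Physics asset for the AirKringle pickaxe axe mesh, a melee weapon cosmetic.
# It defines a single capsule body on the b_AR_Root bone, long and thin and offset from the bone origin.
# With no constraints, the axe collides as one rigid piece with no articulated parts.
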PhysicsAsset /BRCosmetics/Weapons/FORT_Melee/Pickaxe_AirKringle/Meshes/AirKringle_Axe_PhysicsAsset
bodies {b_AR_Root [{shapeType=Capsule, length=46.7438, radius=4.5685, center=<0.7611 -0.4607 24.891>, rotation=<rot -0.6831 -90 -0.6038>}]}
constraints []
